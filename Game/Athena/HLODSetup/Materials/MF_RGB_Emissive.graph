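# Material function in the Athena HLODSetup folder that turns a material's emissive into an animated rainbow by cycling its hue across space and time.
# Summed coordinate components, world position by default, minus scaled time feed a sine, which is offset and scaled into a hue-shift curve.
# The shifted, intensity-scaled emissive is lerped in by the UseRGB? scalar toggle.
# Extra outputs expose the curve, a base hue color and that color hue-shifted.
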MaterialFunction /Game/Athena/HLODSetup/Materials/MF_RGB_Emissive
graph {nodes [
  n1 = GetMaterialAttributes(None=FunctionInput() {inputName=MA, inputType=FunctionInput_MaterialAttributes}) {attributeGetTypes [#1]}
  n2 = ScalarParameter() {
   controlType=Enumeration
   enumeration=/DealershipSharedLibrary/FX/Materials/Enums/Enum_DLR_TrueFalse
   group=RGB Emissive
   parameterName=UseRGB?
  }
  n3 = WorldPosition()
  n4 = MaterialFunctionCall(f3=TruncateLWC!(None=FunctionInput(Preview=@n3.XYZ) {bUsePreviewValueAsDefault=true, inputName=Coords})) {materialFunction=&AddComponents}
  n5 = ScalarParameter() {defaultValue=13000, group=RGB Emissive, parameterName=HUE Curve Lenght}
  n6 = Time()
  n7 = ScalarParameter() {defaultValue=0.2, group=RGB Emissive, parameterName=RGB Time}
  n8 = Sine(None=Subtract(A=Divide(A=@n4.f3, B=@n5), B=Multiply(A=@n6, B=@n7) {constB=0.2}))
  n9 = ScalarParameter() {defaultValue=-0.7, group=RGB Emissive, parameterName=HUE Curve Position}
  n10 = ComponentMask(None=Multiply(A=Add(A=@n8, B=@n9), B=ScalarParameter() {defaultValue=0.1, group=RGB Emissive, parameterName=HUE Curve Size}) {constB=0.5}) {r=true}
  n11 = Multiply(
   A=MaterialFunctionCall(Hue Shift Percentage=NamedRerouteDeclaration(None=@n10) {name=Curve, nodeColor=<rgba 1 0.7031 0 1>}, Texture=@n1.EmissiveColor) {materialFunction=&HueShift}.Result
   B=ScalarParameter() {defaultValue=1, group=RGB Emissive, parameterName=Emissive Int rgb}
  )
  FunctionOutput(None=SetMaterialAttributes!(MaterialAttributes=@n1.MaterialAttributes, EmissiveColor=LinearInterpolate(A=@n1.EmissiveColor, B=@n11, Alpha=@n2)) {attributeSetTypes [#1]}) {outputName="Rainbow Emissive (MA)"}
  FunctionOutput(None=NamedRerouteUsage()) {outputName="Final Curve Shift (S)", sortPriority=2}
  n12 = VectorParameter() {group=RGB Emissive, parameterName=Base HUE, defaultValue=<rgba 1 0 0 1>}
  FunctionOutput(None=MaterialFunctionCall(Hue Shift Percentage=NamedRerouteUsage(), Texture=@n12.RGB) {materialFunction=&HueShift}.Result) {outputName="Final Rainbow Color (V3)", sortPriority=3}
  FunctionOutput(None=@n12.RGB) {outputName="Base Color (V3)", sortPriority=1}
 ]}
