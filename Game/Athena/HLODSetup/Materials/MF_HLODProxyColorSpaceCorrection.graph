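# Material function, exposed under the HLOD library category, that adjusts color space when a material is baked for an HLOD proxy.
# The Realtime input passes through unchanged in normal rendering via a Material Proxy Replace node.
# For proxy generation, it is split into three scalars, each run through the engine's LinearTosRGB function, then recombined into a Vector3 feeding the proxy branch.
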MaterialFunction /Game/Athena/HLODSetup/Materials/MF_HLODProxyColorSpaceCorrection
graph {
 expressions {
  MaterialExpressionConvert_0 {
   class=MaterialExpressionConvert
   inputs [None {from=MaterialExpressionFunctionInput_0}]
   properties {
    nodeName=Break Float3
    convertInputs [{
      type=Vector3
      expressionInput {
       expression=&self:MaterialExpressionFunctionInput_0
       inputName=None
       mask=0
       maskA=0
       maskB=0
       maskG=0
       maskR=0
       outputIndex=0
      }
     }]
    convertOutputs [{type=Scalar}, {type=Scalar}, {type=Scalar}]
   }
  }
  MaterialExpressionConvert_1 {
   class=MaterialExpressionConvert
   inputs [
    None {from=MaterialExpressionMaterialFunctionCall_4, output=sRGB}
    None {from=MaterialExpressionMaterialFunctionCall_6, output=sRGB}
    None {from=MaterialExpressionMaterialFunctionCall_8, output=sRGB}
   ]
   properties {
    nodeName=Make Vector3
    convertInputs [
     {
      type=Scalar
      expressionInput {
       expression=&self:MaterialExpressionMaterialFunctionCall_4
       inputName=None
       mask=0
       maskA=0
       maskB=0
       maskG=0
       maskR=0
       outputIndex=0
      }
     }
     {
      type=Scalar
      expressionInput {
       expression=&self:MaterialExpressionMaterialFunctionCall_6
       inputName=None
       mask=0
       maskA=0
       maskB=0
       maskG=0
       maskR=0
       outputIndex=0
      }
     }
     {
      type=Scalar
      expressionInput {
       expression=&self:MaterialExpressionMaterialFunctionCall_8
       inputName=None
       mask=0
       maskA=0
       maskB=0
       maskG=0
       maskR=0
       outputIndex=0
      }
     }
    ]
    convertOutputs [{type=Vector3}]
   }
  }
  MaterialExpressionFunctionInput_0 {class=MaterialExpressionFunctionInput, inputs [Preview {from="", output=""}], properties {inputName=Realtime}}
  MaterialExpressionFunctionOutput_0 {class=MaterialExpressionFunctionOutput, inputs [None {from=MaterialExpressionMaterialProxyReplace_2, output=""}]}
  MaterialExpressionMaterialFunctionCall_4 {
   class=MaterialExpressionMaterialFunctionCall
   inputs [Linear {from=MaterialExpressionConvert_0}]
   properties {materialFunction=/Engine/Functions/Engine_MaterialFunctions01/ImageAdjustment/LinearTosRGB}
  }
  MaterialExpressionMaterialFunctionCall_6 {
   class=MaterialExpressionMaterialFunctionCall
   inputs [Linear {from=MaterialExpressionConvert_0}]
   properties {materialFunction=&LinearTosRGB}
  }
  MaterialExpressionMaterialFunctionCall_8 {
   class=MaterialExpressionMaterialFunctionCall
   inputs [Linear {from=MaterialExpressionConvert_0}]
   properties {materialFunction=&LinearTosRGB}
  }
  MaterialExpressionMaterialProxyReplace_2 {
   class=MaterialExpressionMaterialProxyReplace
   inputs [
    Realtime {from=MaterialExpressionFunctionInput_0, output=""}
    MaterialProxy {from=MaterialExpressionConvert_1, output=""}
   ]
   properties {}
  }
 }
 properties {bExposeToLibrary=true, libraryCategoriesText [HLOD]}
}
